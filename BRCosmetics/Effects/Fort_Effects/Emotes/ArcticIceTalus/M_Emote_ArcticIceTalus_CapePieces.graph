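# Unlit, translucent, two-sided material for cape pieces in the ArcticIceTalus emote, built for Niagara mesh particles.
# Dynamic parameters drive erosion that blends panning cloud noise with a sine-shaped UV1 mask, hard-cut by Ceil.
# Emissive color is particle color times a base color broken up by rounded blurred noise.
# Side chains build a flattened bandage normal and a panning Perlin offset vector scaled by V, resembling cloth flutter.
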Material /BRCosmetics/Effects/Fort_Effects/Emotes/ArcticIceTalus/M_Emote_ArcticIceTalus_CapePieces
graph {
 properties {
  bUsedWithNiagaraMeshParticles=true
  blendMode=BLEND_Translucent
  pixelDepthOffsetMode=PDOM_Legacy
  shadingModel=MSM_Unlit
  twoSided=true
  usedShadingModels=Unlit
 }
 nodes [
  n1 = DynamicParameter() {paramNames [Power, UV_Offset, "Erode(1-0)", ErodeItensity]}
  n2 = Multiply(A=Subtract(B=@n1."Erode(1-0)"), B=@n1.ErodeItensity)
  n3 = VectorParameter() {parameterName=DistortUV_ctrl, defaultValue=<rgba 1 0.5 0 0>}
  n4 = Panner(Coordinate=TextureCoordinate() {coordinateIndex=1}, Speed=AppendVector(A=@n3.B, B=@n3.B))
  n5 = AppendVector(A=@n3.R, B=@n3.R)
  n6 = TextureSample(UVs=Add(A=@n1.UV_Offset, B=Multiply(A=@n4, B=@n5))) {texture=&AtmosphericCloudNoise01}
  n7 = TextureCoordinate() {coordinateIndex=1}
  n8 = MaterialFunctionCall(Float2=Sine(None=Multiply(A=@n7) {constB=0.5})) {materialFunction=&BreakOutFloat2Components}
  n9 = Multiply(A=@n8.R, B=@n8.R)
  n10 = Power(Base=@n9, Exp=@n1.Power) {constExponent=5}
  n11 = LinearInterpolate(
   A=LinearInterpolate(A=Subtract(B=@n2) {constA=0}, B=Subtract(B=@n2), Alpha=@n6.R)
   B=Multiply(A=@n9, B=LinearInterpolate(A=@n6.R, B=@n9, Alpha=@n10))
   Alpha=@n1."Erode(1-0)"
  )
  n12 = ScalarParameter() {parameterName=EnhanceErode}
  Ceil(None=Subtract(A=Add(A=@n11, B=@n12)) {constB=0.333})
  n13 = TextureSample() {samplerType=SAMPLERTYPE_Grayscale, texture=&LowResBlurredNoise}
  n14 = LinearInterpolate(A=VectorParameter() {parameterName=base color, defaultValue=<rgba 0.3271 0.3748 0.4 1>}.RGB, Alpha=Round(None=Multiply(A=@n13.R) {constB=2})) {constA=0.3}
  MaterialFunctionCall(Emissive=Multiply(A=ParticleColor().RGB, B=Saturate(None=@n14))) {materialFunction=&MF_FX_EmissiveAdjustments}
  n15 = MaterialFunctionCall(UVs=Add(A=Constant2Vector() {r=-0.1}, B=TextureCoordinate()), Texture Scale=Constant2Vector() {g=1, r=4.5}) {materialFunction=&ScaleUVsByCenter}
  Multiply(
   A=MaterialFunctionCall(
    Normal=TextureSample(UVs=@n15.UVs) {samplerType=SAMPLERTYPE_Normal, texture=/Game/Items/Art_noLOD/Textures/T_MedBandage_N}.RGB
    Flatness=Constant() {r=0.9}
   ) {materialFunction=&FlattenNormal}.Result
   B=ScalarParameter() {defaultValue=1, parameterName=Normal}
  ) {constB=0}
  n16 = TextureCoordinate() {uTiling=0.2, vTiling=0.4}
  n17 = Time()
  n18 = DynamicParameter() {paramNames [Power, UV_Offset, "Erode(1-0)", ErodeItensity]}
  n19 = TextureSample(UVs=Panner(Coordinate=@n16, Time=Add(A=@n17, B=@n18.UV_Offset)) {speedX=0.3, speedY=0.3}) {texture=/Game/Global/Textures/Default/Noise/3xPerlin_M}
  n20 = ComponentMask(None=@n19.RGB) {g=true, r=true}
  n21 = Multiply(A=@n19.B) {constB=-3}
  Multiply(A=Multiply(
    A=AppendVector(A=@n20, B=@n21)
    B=MaterialFunctionCall(Float2=@n16) {materialFunction=&BreakOutFloat2Components}.G
   )) {constB=40}
 ]
}
